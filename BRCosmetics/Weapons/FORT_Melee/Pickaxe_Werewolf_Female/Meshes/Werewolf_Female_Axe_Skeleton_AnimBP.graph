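# Animation Blueprint for the female werewolf pickaxe that adds idle sway and springy secondary motion to three dynamic bone chains.
# Each update, the EventGraph feeds game time through RandomRotatorFunction with staggered settings into six rotator variables.
# The AnimGraph copies the mesh pose and additively applies those rotators to the first and third bones of each chain.
# Root-relative, zero-gravity Anim Dynamics with angular springs and limits then simulates the second and fourth bones.
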AnimBlueprint /BRCosmetics/Weapons/FORT_Melee/Pickaxe_Werewolf_Female/Meshes/Werewolf_Female_Axe_Skeleton_AnimBP
target_skeleton=&/Werewolf_Female_Axe_Skeleton
graphs [
 AnimGraph {nodes [
   n1 = "Transform (Modify) Bone - Bone: dyn_1_1"(
    ComponentPose=Local To Component(LocalPose=Copy Pose From Mesh())
    Alpha=1
    Translation="0,0,0"
    Rotation=Get Rotator_1_1() -> (Rotator_1_1:Rotator)
    Scale="1,1,1"
   ) {
    rotationMode=BMM_Additive
    rotationSpace=BCS_ParentBoneSpace
    translationMode=BMM_Additive
    translationSpace=BCS_ParentBoneSpace
    alphaBoolBlend {blendOption=Linear}
    boneToModify {boneName=dyn_1_1}
   }
   n2 = "Transform (Modify) Bone - Bone: dyn_1_3"(ComponentPose=@n1, Alpha=1, Translation="0,0,0", Rotation=Get Rotator_1_2() -> (Rotator_1_2:Rotator), Scale="1,1,1") {
    rotationMode=BMM_Additive
    rotationSpace=BCS_ParentBoneSpace
    translationMode=BMM_Additive
    translationSpace=BCS_ParentBoneSpace
    alphaBoolBlend {blendOption=Linear}
    boneToModify {boneName=dyn_1_3}
   }
   n3 = "Transform (Modify) Bone - Bone: dyn_2_1"(ComponentPose=@n2, Alpha=1, Translation="0,0,0", Rotation=Get Rotator_2_1() -> (Rotator_2_1:Rotator), Scale="1,1,1") {
    rotationMode=BMM_Additive
    rotationSpace=BCS_ParentBoneSpace
    translationMode=BMM_Additive
    translationSpace=BCS_ParentBoneSpace
    alphaBoolBlend {blendOption=Linear}
    boneToModify {boneName=dyn_2_1}
   }
   n4 = "Transform (Modify) Bone - Bone: dyn_2_3"(ComponentPose=@n3, Alpha=1, Translation="0,0,0", Rotation=Get rotator_2_2() -> (rotator_2_2:Rotator), Scale="1,1,1") {
    rotationMode=BMM_Additive
    rotationSpace=BCS_ParentBoneSpace
    translationMode=BMM_Additive
    translationSpace=BCS_ParentBoneSpace
    alphaBoolBlend {blendOption=Linear}
    boneToModify {boneName=dyn_2_3}
   }
   n5 = "Transform (Modify) Bone - Bone: dyn_3_1"(ComponentPose=@n4, Alpha=1, Translation="0,0,0", Rotation=Get Rotator_3_1() -> (Rotator_3_1:Rotator), Scale="1,1,1") {
    rotationMode=BMM_Additive
    rotationSpace=BCS_ParentBoneSpace
    translationMode=BMM_Additive
    translationSpace=BCS_ParentBoneSpace
    alphaBoolBlend {blendOption=Linear}
    boneToModify {boneName=dyn_3_1}
   }
   n6 = "Transform (Modify) Bone - Bone: dyn_3_3"(ComponentPose=@n5, Alpha=1, Translation="0,0,0", Rotation=Get Rotator_3_2() -> (Rotator_3_2:Rotator), Scale="1,1,1") {
    rotationMode=BMM_Additive
    rotationSpace=BCS_ParentBoneSpace
    translationMode=BMM_Additive
    translationSpace=BCS_ParentBoneSpace
    alphaBoolBlend {blendOption=Linear}
    boneToModify {boneName=dyn_3_3}
   }
   n7 = Anim Dynamics - Bone: dyn_1_2(ComponentPose=@n6, Alpha=1, ExternalForce="0,0,0") {
    angularDampingOverride=1
    angularSpringConstant=20000
    bAngularSpring=true
    bGravityOverrideInSimSpace=true
    bOverrideAngularDamping=true
    gravityScale=0
    simulationSpace=RootRelative
    alphaBoolBlend {blendOption=Linear}
    boundBone {boneName=dyn_1_2}
    componentLinearAccScale=<0.5 0.5 0.5>
    componentLinearVelScale=<1 1 1>
    gravityOverride {z=-980}
    physicsBodyDefinitions [{
      sphereCollisionRadius=0
      boundBone {boneName=dyn_1_2}
      boxExtents=<22 32 32>
      constraintSetup {angularLimitsMax=<15 15 15>, angularLimitsMin=<-15 -15 -15>, angularTarget=<1 0 0>}
      localJointOffset=<0 0 -4>
     }]
    relativeSpaceBone {boneName=b_AR_Root}
   }
   n8 = Anim Dynamics - Bone: dyn_1_4(ComponentPose=@n7, Alpha=1, ExternalForce="0,0,0") {
    angularDampingOverride=1
    angularSpringConstant=20000
    bAngularSpring=true
    bGravityOverrideInSimSpace=true
    bOverrideAngularDamping=true
    gravityScale=0
    simulationSpace=RootRelative
    alphaBoolBlend {blendOption=Linear}
    boundBone {boneName=dyn_1_4}
    componentLinearAccScale=<0.25 0.25 0.25>
    componentLinearVelScale=<1 1 1>
    gravityOverride {z=-980}
    physicsBodyDefinitions [{
      sphereCollisionRadius=0
      boundBone {boneName=dyn_1_4}
      boxExtents=<22 32 32>
      constraintSetup {angularLimitsMax=<15 15 15>, angularLimitsMin=<-15 -15 -15>, angularTarget=<1 0 0>}
      localJointOffset=<0 0 -4>
     }]
    relativeSpaceBone {boneName=b_AR_Root}
   }
   n9 = Anim Dynamics - Bone: dyn_2_2(ComponentPose=@n8, Alpha=1, ExternalForce="0,0,0") {
    angularDampingOverride=1
    angularSpringConstant=20000
    bAngularSpring=true
    bGravityOverrideInSimSpace=true
    bOverrideAngularDamping=true
    gravityScale=0
    simulationSpace=RootRelative
    alphaBoolBlend {blendOption=Linear}
    boundBone {boneName=dyn_2_2}
    componentLinearAccScale=<0.5 0.5 0.5>
    componentLinearVelScale=<1 1 1>
    gravityOverride {z=-980}
    physicsBodyDefinitions [{
      sphereCollisionRadius=0
      boundBone {boneName=dyn_2_2}
      boxExtents=<22 32 32>
      constraintSetup {angularLimitsMax=<15 15 15>, angularLimitsMin=<-15 -15 -15>, angularTarget=<1 0 0>}
      localJointOffset=<0 0 -4>
     }]
    relativeSpaceBone {boneName=b_AR_Root}
   }
   n10 = Anim Dynamics - Bone: dyn_2_4(ComponentPose=@n9, Alpha=1, ExternalForce="0,0,0") {
    angularDampingOverride=1
    angularSpringConstant=20000
    bAngularSpring=true
    bGravityOverrideInSimSpace=true
    bOverrideAngularDamping=true
    gravityScale=0
    simulationSpace=RootRelative
    alphaBoolBlend {blendOption=Linear}
    boundBone {boneName=dyn_2_4}
    componentLinearAccScale=<0.25 0.25 0.25>
    componentLinearVelScale=<1 1 1>
    gravityOverride {z=-980}
    physicsBodyDefinitions [{
      sphereCollisionRadius=0
      boundBone {boneName=dyn_2_4}
      boxExtents=<22 32 32>
      constraintSetup {angularLimitsMax=<15 15 15>, angularLimitsMin=<-15 -15 -15>, angularTarget=<1 0 0>}
      localJointOffset=<0 0 -4>
     }]
    relativeSpaceBone {boneName=b_AR_Root}
   }
   n11 = Anim Dynamics - Bone: dyn_3_2(ComponentPose=@n10, Alpha=1, ExternalForce="0,0,0") {
    angularDampingOverride=1
    angularSpringConstant=20000
    bAngularSpring=true
    bGravityOverrideInSimSpace=true
    bOverrideAngularDamping=true
    gravityScale=0
    simulationSpace=RootRelative
    alphaBoolBlend {blendOption=Linear}
    boundBone {boneName=dyn_3_2}
    componentLinearAccScale=<0.5 0.5 0.5>
    componentLinearVelScale=<1 1 1>
    gravityOverride {z=-980}
    physicsBodyDefinitions [{
      sphereCollisionRadius=0
      boundBone {boneName=dyn_3_2}
      boxExtents=<22 32 32>
      constraintSetup {angularLimitsMax=<15 15 15>, angularLimitsMin=<-15 -15 -15>, angularTarget=<1 0 0>}
      localJointOffset=<0 0 -4>
     }]
    relativeSpaceBone {boneName=b_AR_Root}
   }
   n12 = Anim Dynamics - Bone: dyn_3_4(ComponentPose=@n11, Alpha=1, ExternalForce="0,0,0") {
    angularDampingOverride=1
    angularSpringConstant=20000
    bAngularSpring=true
    bGravityOverrideInSimSpace=true
    bOverrideAngularDamping=true
    gravityScale=0
    simulationSpace=RootRelative
    alphaBoolBlend {blendOption=Linear}
    boundBone {boneName=dyn_3_4}
    componentLinearAccScale=<0.25 0.25 0.25>
    componentLinearVelScale=<1 1 1>
    gravityOverride {z=-980}
    physicsBodyDefinitions [{
      sphereCollisionRadius=0
      boundBone {boneName=dyn_3_4}
      boxExtents=<22 32 32>
      constraintSetup {angularLimitsMax=<15 15 15>, angularLimitsMin=<-15 -15 -15>, angularTarget=<1 0 0>}
      localJointOffset=<0 0 -4>
     }]
    relativeSpaceBone {boneName=b_AR_Root}
   }
   Output Pose(Result=Component To Local(ComponentPose=@n12)) {layerGroup=None}
  ]}
 EventGraph {nodes [
   Event BlueprintUpdateAnimation()
   n1 = GetGameTimeInSeconds()
   > n2 = RandomRotatorFunction!(
    execute
    GameTimeInSeconds:double=@n1
    PerlinScale:double=0
    Frequency:Rotator="1, 0, 0"
    Amplitude:Rotator="2,0, 0"
    Offset:Rotator="0, 0, 0"
    TimeOffset:double=0
   ) -> (then, RandomRotator:Rotator)
   > Set Rotator_1_1(Rotator_1_1:Rotator=@n2) -> (Output_Get:Rotator)
   > n3 = RandomRotatorFunction(GameTimeInSeconds=@n1, PerlinScale=0, Frequency="1, 0, 0", Amplitude="3, 0, 0", Offset="0, 0, 0", TimeOffset=2)
   > Set Rotator_1_2(Rotator_1_2:Rotator=@n3) -> (Output_Get:Rotator)
   > n4 = RandomRotatorFunction(GameTimeInSeconds=@n1, PerlinScale=0, Frequency="-1, 0, 0", Amplitude="1.25, 0, 0", Offset="0, 0, 0", TimeOffset=0)
   > Set Rotator_2_1(Rotator_2_1:Rotator=@n4) -> (Output_Get:Rotator)
   > n5 = RandomRotatorFunction(GameTimeInSeconds=@n1, PerlinScale=0, Frequency="-1, 0, 0", Amplitude="-3, 0, 0", Offset="0, 0, 0", TimeOffset=2.5)
   > Set Rotator_2_2(Rotator_2_2:Rotator=@n5) -> (Output_Get:Rotator)
   > n6 = RandomRotatorFunction(GameTimeInSeconds=@n1, PerlinScale=0, Frequency="0.75, 0, 0", Amplitude="2.5, 0, 0", Offset="0, 0, 0", TimeOffset=1)
   > Set Rotator_3_1(Rotator_3_1:Rotator=@n6) -> (Output_Get:Rotator)
   > n7 = RandomRotatorFunction(GameTimeInSeconds=@n1, PerlinScale=0, Frequency="0.9, 0, 0", Amplitude="3, 0, 0", Offset="0, 0, 0", TimeOffset=3.5)
   > Set Rotator_3_2(Rotator_3_2:Rotator=@n7) -> (Output_Get:Rotator)
  ]}
]
variables [
 Rotator_1_1 {type=Rotator}
 Rotator_1_2 {type=Rotator}
 Rotator_2_1 {type=Rotator}
 Rotator_2_2 {type=Rotator}
 Rotator_3_1 {type=Rotator}
 Rotator_3_2 {type=Rotator}
]
